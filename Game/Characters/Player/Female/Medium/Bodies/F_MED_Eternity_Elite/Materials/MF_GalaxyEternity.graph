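# Material function that blends an animated galaxy surface over incoming material attributes by an alpha input, for the Eternity Elite character.
# Camera-vector triplanar sampling projects a rotating, warp-maskable nebula and two star layers, with tiling scaled by FOV.
# Panning star-texture noise twinkles the large stars.
# A fresnel blends in a noise-driven accent rim, feeding emissive over black base color with adjustable specular and roughness.
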MaterialFunction /Game/Characters/Player/Female/Medium/Bodies/F_MED_Eternity_Elite/Materials/MF_GalaxyEternity
graph {nodes [
  n1 = FunctionInput() {inputType=FunctionInput_MaterialAttributes}
  n2 = GetMaterialAttributes(None=@n1) {attributeGetTypes [#1]}
  n3 = ScalarParameter() {defaultValue=0.5, group=Galaxy, parameterName=GalaxyTiling}
  n4 = MaterialFunctionCall() {materialFunction=/Engine/Functions/Engine_MaterialFunctions02/Utility/FOV}
  n5 = Constant() {r=0.414}
  n6 = Constant() {r=1.091}
  n7 = Constant()
  n8 = Constant() {r=1}
  n9 = LinearInterpolate(
   A=Constant() {r=3}
   B=Constant() {r=1}
   Alpha=Saturate(None=MaterialFunctionCall(Input=@n4."FOV ", Input Low=@n5, Input High=@n6, Target Low=@n7, Target High=@n8) {materialFunction=&RemapValueRange}.Result)
  )
  n10 = FunctionInput() {bUsePreviewValueAsDefault=true, inputName=Warp Mask, inputType=FunctionInput_Scalar}
  n11 = ScalarParameter() {defaultValue=0.012, group=Galaxy, parameterName=WarpIntensity}
  n12 = Time()
  n13 = LinearInterpolate(A=@n12, B=Multiply(A=@n12, B=Constant()), Alpha=ScalarParameter() {group=Galaxy, parameterName=Background Rotation})
  n14 = AppendVector(
   A=AppendVector(
    A=ScalarParameter() {defaultValue=0.02, group=Galaxy, parameterName=Galaxy Rotation Speed}
    B=ScalarParameter() {defaultValue=-0.02, group=Galaxy, parameterName=Galaxy Rotation Speed_1}
   )
   B=ScalarParameter() {defaultValue=0.1, group=Galaxy, parameterName=Galaxy Rotation Speed_2}
  )
  n15 = TextureObjectParameter() {
   group=Galaxy
   parameterName=GalaxyTexture
   texture=/Game/Characters/Player/Male/Medium/Bodies/M_Med_Soldier_04/Skins/Galaxy/Textures/T_FN_Nebula
  }
  n16 = Multiply(
   A=VectorParameter() {group=Galaxy, parameterName=GalaxyNebulaColorMult, defaultValue=<rgba 1 1 1 1>}.RGB
   B=MaterialFunctionCall!(
    CameraVector
    Rotation Speed=Multiply(A=@n13, B=@n14)
    Texture=@n15
    Tiling=Multiply(A=@n3, B=@n9)
    Offset=Multiply(A=@n10, B=@n11)
    AxisFadeContrast=Constant() {r=6}
   ) {materialFunction=/Game/Weapons/WeaponSkins/MaterialFunction/TriplanarCameraVector_Rotated}.Result
  )
  n17 = Desaturation(None=@n16, Fraction=ScalarParameter() {defaultValue=0.5, group=Galaxy, parameterName=Star Desaturation})
  n18 = VectorParameter() {group=Galaxy, parameterName=Small Star Brightness, defaultValue=<rgba 300 300 300 1>}
  n19 = Time()
  n20 = ScalarParameter() {defaultValue=-0.125, group=Galaxy, parameterName=Small Star Rotation speed}
  n21 = TextureObject() {samplerType=SAMPLERTYPE_Masks, texture=/Game/Characters/Player/Male/Medium/Bodies/M_Med_Soldier_04/Skins/Galaxy/Textures/Celestial_T_Stars}
  n22 = ScalarParameter() {defaultValue=5, group=Galaxy, parameterName=Small Star Tiling}
  n23 = ComponentMask(None=MaterialFunctionCall(Rotation Speed=Multiply(A=@n19, B=@n20), Texture=@n21, Tiling=Multiply(A=@n22, B=@n9)) {materialFunction=&TriplanarCameraVector_Rotated}.Result) {b=true}
  n24 = Multiply(A=@n16, B=ScalarParameter() {defaultValue=1, group=Galaxy, parameterName=Galaxy_Brightness})
  n25 = TextureObject() {samplerType=SAMPLERTYPE_Masks, texture=&Celestial_T_Stars}
  n26 = ScalarParameter() {defaultValue=2, group=Galaxy, parameterName=Bright Star Tiling}
  n27 = ComponentMask(None=MaterialFunctionCall!(CameraVector, Texture=@n25, Tiling=Multiply(A=@n26, B=@n9), Offset, AxisFadeContrast) {materialFunction=/Engine/Functions/Engine_MaterialFunctions01/Texturing/TriplanarCameraVector}.Result) {r=true}
  n28 = TextureSample(UVs=Panner(Coordinate=TextureCoordinate() {uTiling=2, vTiling=2}) {speedX=0.133, speedY=0.2}) {
   samplerSource=SSM_Wrap_WorldGroupSettings
   samplerType=SAMPLERTYPE_Masks
   texture=&Celestial_T_Stars
  }
  n29 = LinearInterpolate(
   A=LinearInterpolate(A=@n24, B=Multiply(A=@n17, B=@n18.RGB), Alpha=Multiply(A=@n23, B=@n23))
   B=Multiply(A=@n17, B=VectorParameter() {group=Galaxy, parameterName=Large Star Brightness, defaultValue=<rgba 950 950 950 1>}.RGB)
   Alpha=Multiply(A=Multiply(A=@n27, B=@n27), B=@n28.G)
  )
  n30 = TextureCoordinate() {uTiling=2, vTiling=2}
  n31 = ScalarParameter() {defaultValue=1, group=Galaxy, parameterName=Edge Fresnel Tiling}
  n32 = ScalarParameter() {group=Galaxy, parameterName=GradientSpeed_X}
  n33 = ScalarParameter() {defaultValue=0.1, group=Galaxy, parameterName=GradientSpeed_Y}
  n34 = TextureSample(UVs=Panner(Coordinate=Multiply(A=@n30, B=@n31), Speed=AppendVector(A=@n32, B=@n33))) {
   samplerSource=SSM_Wrap_WorldGroupSettings
   samplerType=SAMPLERTYPE_LinearColor
   texture=/Game/Weapons/WeaponSkins/Textures/T_Swirly_Noise_highContrast1
  }
  n35 = LinearInterpolate(
   A=Multiply(A=ScalarParameter() {defaultValue=8, group=Galaxy, parameterName=AccentRimBrightness}, B=@n24)
   B=VectorParameter() {group=Galaxy, parameterName=EdgeFresnel, defaultValue=<rgba 5.0909 8.586 10 1>}.RGB
   Alpha=Multiply(
    A=FunctionInput(Preview=@n34.R) {bUsePreviewValueAsDefault=true, inputName=GalaxyAlternateFresnel, inputType=FunctionInput_Scalar}
    B=ScalarParameter() {defaultValue=1, group=Galaxy, parameterName=Galaxy Flame Fresnel Intensity}
   )
  )
  n36 = Saturate(None=Fresnel(ExponentIn=ScalarParameter() {defaultValue=5, group=Galaxy, parameterName=Galaxy_FresEX}) {baseReflectFraction=0})
  n37 = SetMaterialAttributes!(
   MaterialAttributes=@n2.MaterialAttributes
   BaseColor=Constant()
   Specular=ScalarParameter() {group=Galaxy, parameterName=Galaxy Specular}
   Roughness=Multiply(A=@n2.Roughness, B=ScalarParameter() {defaultValue=1, group=Galaxy, parameterName=Galaxy Roughness})
   EmissiveColor=Multiply(A=ScalarParameter() {defaultValue=1, group=Galaxy, parameterName=Galaxy_GlobalBrightness}, B=LinearInterpolate(A=@n29, B=@n35, Alpha=@n36))
  ) {attributeSetTypes [
    #2
    #3
    #1
    #4
   ]}
  FunctionOutput(None=BlendMaterialAttributes(A=@n1, B=@n37, Alpha=FunctionInput() {inputName=Alpha, inputType=FunctionInput_Scalar}) {vertexAttributeBlendType=UseA})
  ScalarParameter() {defaultValue=6, group=Galaxy, parameterName=GalaxyAxisFade}
 ]}
