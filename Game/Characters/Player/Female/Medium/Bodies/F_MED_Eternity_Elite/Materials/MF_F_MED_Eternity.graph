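# Material function for the Eternity Elite female body, adding an animated galaxy look to incoming material attributes.
# Switchable panned or projected swirly noise plus offset sine waves form a masked UV ripple distorting diffuse, specular, emissive and FX mask samples.
# Specular masks, FX mask and noise feed MF_GalaxyEternity.
# Its output gains emissive, noise-driven rainbow and star glow, while Fresnel terms darken base color, more adjustably in masked areas.
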MaterialFunction /Game/Characters/Player/Female/Medium/Bodies/F_MED_Eternity_Elite/Materials/MF_F_MED_Eternity
graph {nodes [
  n1 = FunctionInput() {inputType=FunctionInput_MaterialAttributes}
  n2 = TextureObject() {texture=/Game/Weapons/WeaponSkins/Textures/T_Swirly_Noise_D1}
  n3 = VectorParameter() {group=Galaxy, parameterName=GalaxyEF, channelNames {a="", b=GspeedY, g=GspeedX, r=EdgeFresTile}, defaultValue=<rgba 1 0 0.1 1>}
  n4 = AppendVector(A=@n3.GspeedX, B=@n3.GspeedX)
  n5 = TextureCoordinate() {uTiling=2, vTiling=2}
  n6 = TextureSample(UVs=Panner(Coordinate=Multiply(A=@n5, B=@n3.EdgeFresTile), Speed=@n4)) {
   samplerSource=SSM_Wrap_WorldGroupSettings
   samplerType=SAMPLERTYPE_LinearColor
   texture=/Game/Weapons/WeaponSkins/Textures/T_Swirly_Noise_highContrast1
  }
  n7 = StaticSwitchParameter(
   True=ComponentMask(None=MaterialFunctionCall!(TextureObject=@n2, TextureSize=@n3.EdgeFresTile, WorldPosition, Export Float 4, In=@n4, World Space Normal, ProjectionTransitionContrast) {materialFunction=/Game/Athena/SafeZone/Preskinned_WP_Texture_ScalePan}.XYZ Texture) {r=true}
   False=@n6.R
  ) {group=Galaxy, parameterName=UseLocalEdgeFresnel}
  n8 = Multiply(A=Multiply(A=@n7, B=ScalarParameter() {defaultValue=0.002, group=Eternity, parameterName=UVRippleStrength}) {constB=0.002})
  n9 = Multiply(A=Time(), B=ScalarParameter() {defaultValue=0.05, group=Eternity, parameterName=uvSinWaveSpeed}) {constB=0.05}
  n10 = ScalarParameter() {defaultValue=0.002, group=Eternity, parameterName=uvSinWaveStrength}
  n11 = Add(A=@n8, B=Multiply(A=Sine(None=@n9) {period=0.2}, B=@n10) {constB=0.002})
  n12 = Add(A=@n9) {constB=0.333}
  n13 = Add(A=Multiply(A=@n8) {constB=2}, B=Multiply(A=Sine(None=@n12) {period=0.2}, B=@n10) {constB=0.002})
  n14 = TextureSample() {
   samplerType=SAMPLERTYPE_Masks
   texture=/Game/Characters/Player/Female/Medium/Bodies/F_MED_Eternity_Elite/Textures/T_F_MED_Eternity_Elite_Body_FX2
  }
  n15 = Add(A=TextureCoordinate(), B=Multiply(A=AppendVector(A=@n11, B=@n13), B=@n14.R))
  n16 = TextureSampleParameter2D(UVs=@n15) {
   group=Eternity
   parameterName=SpecularMasks
   samplerType=SAMPLERTYPE_Masks
   sortPriority=0
   texture=/Game/Characters/Player/Female/Medium/Bodies/F_MED_Eternity_Elite/Textures/T_F_MED_Eternity_Elite_Body_S
  }
  n17 = Multiply(A=@n16.R, B=ScalarParameter() {defaultValue=1, group=Eternity, parameterName=EternitySpecMult})
  n18 = TextureSampleParameter2D(UVs=@n15) {
   group=Eternity
   parameterName=Eternity_FX_Mask
   samplerType=SAMPLERTYPE_Masks
   sortPriority=0
   texture=/Game/Characters/Player/Female/Medium/Bodies/F_MED_Eternity_Elite/Textures/T_F_MED_Eternity_Elite_Body_FX
  }
  n19 = Multiply(A=@n18.B)
  n20 = GetMaterialAttributes(None=MaterialFunctionCall!(
    In=SetMaterialAttributes!(MaterialAttributes=@n1, Roughness=@n16.B, Metallic=@n16.B, Specular=@n17) {attributeSetTypes [#1, #2, #3]}
    Warp Mask
    GalaxyAlternateFresnel=@n7
    Alpha=Add(A=@n18.R, B=@n19)
   ) {materialFunction=&/MF_GalaxyEternity}.Result) {attributeGetTypes [#4, #5]}
  n21 = TextureSampleParameter2D(UVs=@n15) {
   group=Eternity
   parameterName=Emissive
   sortPriority=0
   texture=/Game/Characters/Player/Female/Medium/Bodies/F_MED_Eternity_Elite/Textures/T_F_MED_Eternity_Elite_Body_E
  }
  n22 = Add(A=@n20.EmissiveColor, B=@n21.RGB)
  n23 = Multiply(A=Add(A=Multiply(A=@n7) {constB=0.75}) {constB=0.1}, B=@n18.B)
  n24 = TextureSampleParameter2D(UVs=@n15) {
   group=Eternity
   parameterName=Diffuse
   sortPriority=0
   texture=/Game/Characters/Player/Female/Medium/Bodies/F_MED_Eternity_Elite/Textures/T_F_MED_Eternity_Elite_Body_D
  }
  n25 = Add(
   A=Multiply(A=Multiply(A=@n23, B=@n24.RGB), B=ScalarParameter() {defaultValue=0.2, group=Eternity, parameterName=Eternity_RainbowEmissive})
   B=Multiply(A=Multiply(A=@n18.G, B=@n24.RGB), B=ScalarParameter() {defaultValue=0.1, group=Eternity, parameterName=Eternity_StarEmissive}) {constB=0.2}
  )
  n26 = Multiply(A=@n24.RGB, B=OneMinus(None=Fresnel() {baseReflectFraction=0, exponent=1}))
  n27 = Fresnel() {baseReflectFraction=0, exponent=0.5}
  n28 = Power(Base=LinearInterpolate(
    B=OneMinus(None=@n27)
    Alpha=ScalarParameter() {defaultValue=1, group=Eternity, parameterName=Eternity_BaseColorFresnelStrength, sliderMax=1}
   ) {constA=1}) {constExponent=3}
  n29 = Multiply(
   A=Multiply(A=@n26, B=ScalarParameter() {defaultValue=1, group=Eternity, parameterName=Eternity_BaseColorMult, sliderMin=1})
   B=Saturate(None=Multiply(A=@n28) {constB=2})
  )
  FunctionOutput(None=SetMaterialAttributes!(
    MaterialAttributes=@n20.MaterialAttributes
    BaseColor=LinearInterpolate(A=@n26, B=@n29, Alpha=@n18.R) {constAlpha=1}
    EmissiveColor=Add(A=@n22, B=@n25)
   ) {attributeSetTypes [#4, #5]})
 ]}
